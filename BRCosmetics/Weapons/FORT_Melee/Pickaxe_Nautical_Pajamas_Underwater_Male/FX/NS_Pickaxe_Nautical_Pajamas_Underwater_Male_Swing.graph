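# Niagara swing trail effect for the Nautical Pajamas Underwater Male pickaxe, built from a single RibbonTrail emitter with a ribbon renderer.
# After a brief loop delay, it spawns particles at a steady rate, widening the ribbon and setting curve-driven alpha from normalized loop age.
# Each short-lived particle then fades and narrows to zero through alpha and width curves.
# A skeletal mesh user parameter is exposed, but no listed module reads it.
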NiagaraSystem /BRCosmetics/Weapons/FORT_Melee/Pickaxe_Nautical_Pajamas_Underwater_Male/FX/NS_Pickaxe_Nautical_Pajamas_Underwater_Male_Swing
dependencies {
 usedDataInterfaces [&NiagaraDataInterfaceColorCurve, &NiagaraDataInterfaceCurve]
 usedDynamicInputs []
 usedModules [&SystemState]
}
emitters {RibbonTrail {
  emitterUpdateScript [
   &EmitterState(Life Cycle Mode=::NewEnumerator1, Loop Behavior=::NewEnumerator1, Loop Delay=0.05, Loop Duration=0.175, UseLoopDelay=true)
   &SpawnRate(SpawnRate=120, Use Spawn Probability=false)
  ]
  particleSpawnScript [&InitializeRibbon(
    Color=&ColorFromCurve(
     ColorCurve=DataInterface() {propertyValues {
       AlphaCurve {keys [
         {arriveTangent=-0.2001, interpMode=RCIM_Cubic, leaveTangent=-0.2001, tangentMode=RCTM_User, time=0.5, value=0.05}
         {arriveTangent=0.0032, interpMode=RCIM_Cubic, leaveTangent=0.0032, tangentMode=RCTM_User, time=1, value=0}
        ]}
       BlueCurve {keys []}
       GreenCurve {keys []}
       RedCurve {keys []}
      }}
     CurveIndex=$Emitter.NormalizedLoopAge
    )
    Lifetime=0.21
    Module.Write Color=true
    Module.Write Lifetime=true
    Module.Write Mass=true
    Module.Write Material Random=true
    Module.Write Position=false
    Module.Write RibbonTwist=false
    Module.Write RibbonWidth=true
    Ribbon Width=&Lerp_Float(A=5, B=90, Alpha=$Emitter.NormalizedLoopAge)
   )]
  particleUpdateScript [
   &ParticleState()
   &ScaleColor(
    Scale Alpha=&FloatFromCurve(FloatCurve=DataInterface() {propertyValues {Curve {keys [
         {arriveTangent=-0.596, interpMode=RCIM_Cubic, leaveTangent=-0.596, tangentMode=RCTM_User, time=0.1, value=0.6}
         {arriveTangent=-0.0099, interpMode=RCIM_Cubic, leaveTangent=-0.0099, tangentMode=RCTM_User, time=1, value=0}
        ]}}})
    ScaleA=true
    ScaleRGB=true
   )
   &ScaleRibbonWidth(Ribbon Width Scale=&FloatFromCurve(FloatCurve=DataInterface() {propertyValues {Curve {keys [<at 0 1 RCIM_Cubic>, {arriveTangent=-1.8739, interpMode=RCIM_Cubic, leaveTangent=-1.8739, tangentMode=RCTM_User, time=1, value=0}]}}}))
  ]
  rendererClasses [&NiagaraRibbonRendererProperties]
  renderers [{rendererClass=&NiagaraRibbonRendererProperties}]
 }}
summary {
 systemName=NS_Pickaxe_Nautical_Pajamas_Underwater_Male_Swing
 emitters [{emitterName=RibbonTrail, rendererClasses [&NiagaraRibbonRendererProperties]}]
 userVariables [User.SkeletalMesh {
   type=NiagaraDataInterfaceSkeletalMesh
   defaultValue {
    struct=NiagaraExt_VariableValue_DataInterface
    value {dataInterface=&self:NiagaraDataInterfaceSkeletalMesh_0}
   }
  }]
}
