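# Physics asset for the antenna mesh of the Nautical Pajamas Underwater Male pickaxe, giving its eyes, fins, jaw, tail and tendril jointed secondary motion.
# Capsule bodies cover the handle and each dyn_ bone, with eyes, fins, jaw, tail and tendril root constrained to the handle.
# Fins move on one swing axis, the jaw only twists, and the five-link tendril chain keeps swing tight while allowing more twist.
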PhysicsAsset /BRCosmetics/Weapons/FORT_Melee/Pickaxe_Nautical_Pajamas_Underwater_Male/Meshes/Nautical_Pajamas_Underwater_Male_antenna_Physics
bodies {
 Melee_TwoHanded_Handle [
  {shapeType=Capsule, length=106.5271, radius=6.9495, center=<0.7364 -2.4511 0.3094>, rotation=<rot 0.4144 -89.9988 -0.0514>}
  {shapeType=Capsule, length=20, radius=10, center=<0.7318 -6.8833 61.8799>}
  {shapeType=Capsule, length=2.5, radius=8.75, center=<0 -1.5838 75.1423>}
 ]
 dyn_eye_l [{shapeType=Capsule, length=0, radius=4.4894, rotation=<rot 14.4219 -90 -18.9218>}]
 dyn_eye_r [{shapeType=Capsule, length=0, radius=4.4894, rotation=<rot 14.4218 -90 18.9219>}]
 dyn_fin_l_01 [{shapeType=Capsule, length=6.3755, radius=2.75, center=<4.4777 -0.0034 0.406>, rotation=<rot 0.4712 89.3081 -89.8903>}]
 dyn_fin_r_01 [{shapeType=Capsule, length=5.75, radius=2.755, center=<-4.4714 0.3337 -0.3848>, rotation=<rot -0.329 -93.8279 -94.9063>}]
 dyn_jaw [{shapeType=Capsule, length=12.4461, radius=11.0485, center=<-0.419 8.0372 -1.198>, rotation=<rot 2.3631 2.775 -90>}]
 dyn_tail [{shapeType=Capsule, length=10.6499, radius=8.0409, center=<0.0162 -3.5265 2.5626>, rotation=<rot -10.3556 -89.9992 0.0026>}]
 dyn_tendril_01 [{shapeType=Capsule, length=11.0105, radius=3.146, center=<-0.0707 -0.5829 2.2013>, rotation=<rot -0.6798 -90.3127 2.3045>}]
 dyn_tendril_02 [{shapeType=Capsule, length=6.0306, radius=3.0306, center=<-0.1718 -0.7352 4.0106>, rotation=<rot 2.762 -90 -1.68>}]
 dyn_tendril_03 [{shapeType=Capsule, length=5.5366, radius=2.4246, center=<-0.3913 -0.8475 2.4016>, rotation=<rot 0.4852 -89.7286 -1.8079>}]
 dyn_tendril_04 [{shapeType=Capsule, length=6.1031, radius=1.7411, center=<-0.7503 0.2707 4.0213>, rotation=<rot 20.3254 -90 -2.631>}]
 dyn_tendril_05 [{shapeType=Capsule, length=3.9898, radius=3.7716, center=<-0.8171 1.7957 5.6919>, rotation=<rot 23.7959 -90 -0.9717>}]
}
constraints [
 {bone1Name=dyn_eye_l, bone2Name=Melee_TwoHanded_Handle, swing1LimitDegrees=10, swing2LimitDegrees=10, twistLimitDegrees=10}
 {bone1Name=dyn_eye_r, bone2Name=Melee_TwoHanded_Handle, swing1LimitDegrees=10, swing2LimitDegrees=10, twistLimitDegrees=10}
 {bone1Name=dyn_fin_l_01, bone2Name=Melee_TwoHanded_Handle, swing1LimitDegrees=35, swing2LimitDegrees=0, twistLimitDegrees=0}
 {bone1Name=dyn_fin_r_01, bone2Name=Melee_TwoHanded_Handle, swing1LimitDegrees=35, swing2LimitDegrees=0, twistLimitDegrees=0}
 {bone1Name=dyn_jaw, bone2Name=Melee_TwoHanded_Handle, swing1LimitDegrees=0, swing2LimitDegrees=0, twistLimitDegrees=150}
 {bone1Name=dyn_tendril_01, bone2Name=Melee_TwoHanded_Handle, swing1LimitDegrees=1, swing2LimitDegrees=1, twistLimitDegrees=1}
 {bone1Name=dyn_tendril_02, bone2Name=dyn_tendril_01, swing1LimitDegrees=0, swing2LimitDegrees=0, twistLimitDegrees=3}
 {bone1Name=dyn_tendril_03, bone2Name=dyn_tendril_02, swing1LimitDegrees=3, swing2LimitDegrees=3, twistLimitDegrees=12}
 {bone1Name=dyn_tendril_04, bone2Name=dyn_tendril_03, swing1LimitDegrees=1, swing2LimitDegrees=1, twistLimitDegrees=20}
 {bone1Name=dyn_tendril_05, bone2Name=dyn_tendril_04, swing1LimitDegrees=0, swing2LimitDegrees=0, twistLimitDegrees=16}
 {bone1Name=dyn_tail, bone2Name=Melee_TwoHanded_Handle, swing1LimitDegrees=90, swing2LimitDegrees=20, twistLimitDegrees=20}
]
